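# Translucent UI material for the temporary Season 10 Figment minimap base, built on a minimap texture.
# Swizzled, scaled and offset UVs feed the Season 7 holiday snowfall function and a terrain mask sample.
# The snow mask, minus an inverted, contrast-boosted terrain mask alpha, blends the map toward a second texture.
# A blue-tinted map multiply and a minimap mask sample also exist, with no visible consumers.
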Material /Figment_S10_MapUI/M_MiniMap_Figment_MapBase_S10Temp
graph {
 properties {blendMode=BLEND_Translucent, materialDomain=MD_UI, pixelDepthOffsetMode=PDOM_Legacy, usedShadingModels=Default Lit}
 nodes [
  n1 = TextureSample() {texture=&/MiniMapAthena_S10Temp}
  n2 = Add(
   A=MaterialFunctionCall(XY=TextureCoordinate() {uTiling=0.8511, vTiling=-0.8511}) {materialFunction=&Swizzle}.YX
   B=AppendVector(A=Constant() {r=0.03}, B=Constant() {r=-0.01})
  )
  n3 = MaterialFunctionCall!(
   UseHeightLerp
   UV_WorldMask=@n2
   Use_ExtraSnowMask
   LandscapeBlend_Height
   Landscape_RockMask
   Snow_Mask
   Simple_UseWorldZNormal
   Simple_WorldNormal
   WorldPosition
  ) {materialFunction=/Game/Athena/Environments/Landscape/MaterialFunctions/Snow/MF_Athena_Holiday_SnowFall_S7}
  n4 = TextureSample(UVs=@n2) {samplerType=SAMPLERTYPE_LinearColor, texture=/Game/Athena/Prototype/Terrain/Textures/T_Athena_Terrain_Mask}
  n5 = Clamp(None=OneMinus(None=LinearInterpolate(Alpha=@n4.A) {constA=-0.5, constB=1.5}))
  LinearInterpolate(A=@n1.RGB, B=TextureSample().RGB, Alpha=Clamp(None=Subtract(A=@n3.SnowMask, B=@n5)))
  Multiply(A=@n1.RGB, B=Constant3Vector() {constant=<rgba 0.4692 0.8589 1 0>})
  TextureSample() {samplerType=SAMPLERTYPE_Masks, texture=&/MiniMapAthena_S10_MaskTemp}
 ]
}
