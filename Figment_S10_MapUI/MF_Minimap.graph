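# MF_Minimap is a material function in the Figment S10 map UI that supplies minimap color and opacity.
# RGB comes from a linear-color sample of the MapTexture parameter, with optional function inputs to override both texture and UVs.
# Alpha comes from the red channel of the MapOpacityTexture mask parameter, which samples at default UVs and ignores both overrides.
# Both parameters default to temporary Athena S10 minimap textures.
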MaterialFunction /Figment_S10_MapUI/MF_Minimap
graph {nodes [
  n1 = TextureCoordinate()
  n2 = TextureObjectParameter() {parameterName=MapTexture, texture=&/MiniMapAthena_S10Temp}
  FunctionOutput(None=TextureSample(
    UVs=FunctionInput(Preview=@n1) {bUsePreviewValueAsDefault=true, inputName=UVsOverride, inputType=FunctionInput_Vector2}
    Tex=FunctionInput(Preview=@n2) {bUsePreviewValueAsDefault=true, inputName=TextureOverride, inputType=FunctionInput_Texture2D, sortPriority=1}
   ) {samplerType=SAMPLERTYPE_LinearColor}.RGB) {outputName=RGB}
  FunctionOutput(None=TextureSampleParameter2D() {parameterName=MapOpacityTexture, samplerType=SAMPLERTYPE_Masks, texture=&/MiniMapAthena_S10_MaskTemp}.R) {outputName=Alpha, sortPriority=1}
 ]}
