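# Material function for the Apollo map UI that unpacks point-of-interest discoverability from a minimap texture and a scalar Bitmask input.
# It samples the grayscale discoverability texture at an explicit mip level and scales red by 255 to get a bit index.
# A Custom node combines the mask with one shifted by that index using logical `&&`, not bitwise `&`.
# An engine BitMask call and a saturated index-equality test are unconnected to the output.
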MaterialFunction /Game/Athena/Apollo/Maps/UI/MF_DiscoverabilityPOIUnpack
graph {
 properties {bExposeToLibrary=true, libraryCategoriesText [FortniteDiscoverability]}
 nodes [
  n1 = TextureSample!(UVs, Tex, Level, Apply View MipBias) {
   automaticViewMipBias=false
   mipValueMode=TMVM_MipLevel
   samplerType=SAMPLERTYPE_LinearGrayscale
   texture=&/Apollo_Terrain_Minimap_Discoverabilty
  }
  MaterialExpressionMultiply_14 = Multiply(A=@n1.R) {constB=255}
  MaterialExpressionFunctionInput_0 = FunctionInput() {inputName=Bitmask, inputType=FunctionInput_Scalar}
  n2 = Custom!(Bit=@MaterialExpressionMultiply_14, Bitmask=@MaterialExpressionFunctionInput_0) {
   code="return int(Bitmask) && (1 <<  (int(Bit)));"
   outputType=CMOT_Float1
   inputs [
    {inputName=Bit, input {expression=&self:MaterialExpressionMultiply_14}}
    {inputName=Bitmask, input {expression=&self:MaterialExpressionFunctionInput_0}}
   ]
  }
  FunctionOutput(None=@n2)
  MaterialFunctionCall!(Bit=Constant() {r=2}, BitMask) {materialFunction=/Engine/Functions/Engine_MaterialFunctions01/Texturing/BitMask}
  n3 = Constant()
  Saturate(None=If(A=@MaterialExpressionMultiply_14, B=ScalarParameter() {defaultValue=13}, "A > B"=@n3, "A == B"=Constant() {r=1}, "A < B"=@n3))
 ]
}
