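# Translucent UI material for the Figment minimap that animates fog-of-war reveal as a circle expanding from the discovery start location.
# The fog render target's R and G channels, the latter masked by a distance ring driven by DiscoverabilityMPC reveal progress, blend MF_Minimap with MF_Minimap_FOW.
# Panning Perlin noise distorts the fog layer's UVs at the wavefront, where a thin colored line is added.
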Material /Figment_S10_MapUI/M_MinimapFigment_FOW
graph {
 properties {bAutomaticallySetUsageInEditor=false, blendMode=BLEND_Translucent, materialDomain=MD_UI, usedShadingModels=Default Lit}
 nodes [
  n1 = TextureSample() {texture=/Game/Athena/Apollo/Maps/UI/RT_FogofWar}
  n2 = CollectionParameter() {collection=/Game/Athena/Apollo/Maps/UI/DiscoverabilityMPC, parameterName=CurrentStartLocation}
  n3 = ScalarParameter() {defaultValue=0.04, parameterName=SIze}
  n4 = ScalarParameter() {defaultValue=0.6, parameterName=OffsetEnd}
  n5 = Subtract(
   A=Distance(A=TextureCoordinate(), B=ComponentMask(None=@n2) {g=true, r=true})
   B=LinearInterpolate(
    A=Multiply(A=@n3) {constB=-1}
    B=@n4
    Alpha=CollectionParameter() {collection=&DiscoverabilityMPC, parameterName=DiscoverRevealProgress}
   )
  )
  n6 = Saturate(None=Divide(A=@n5, B=@n3))
  n7 = Saturate(None=Add(A=@n1.R, B=Multiply(A=@n1.G, B=@n6)))
  n8 = TextureCoordinate()
  n9 = ScalarParameter() {defaultValue=6, parameterName=DistortScale2}
  n10 = TextureSample(UVs=Panner(Coordinate=Multiply(A=@n8, B=@n9), Speed=ScalarParameter() {defaultValue=0.25, parameterName=DistortionSpeed})) {texture=/Game/Global/Textures/Default/Noise/3xPerlin_M}
  n11 = AppendVector(A=@n10.R, B=@n10.R)
  n12 = ComponentMask(None=VectorParameter() {parameterName=Bias2, defaultValue=<rgba 0.25 0.25 0 1>}.RGB) {g=true, r=true}
  n13 = OneMinus(None=@n6)
  n14 = Multiply(A=Multiply(A=@n1.G, B=Multiply(A=@n6, B=@n13)), B=ScalarParameter() {defaultValue=0.1, parameterName=UVdistortion2})
  n15 = Add(A=Multiply(A=Subtract(A=@n11, B=@n12) {constB=0.5}, B=@n14), B=TextureCoordinate())
  n16 = VectorParameter() {parameterName=LineColor, defaultValue=<rgba 0.05 0.05 0.05 1>}
  n17 = Divide(
   A=Add(A=ScalarParameter() {defaultValue=-0.0491, parameterName=offsetLine}, B=@n5)
   B=ScalarParameter() {defaultValue=0.005, parameterName=SIzeLine}
  )
  n18 = Saturate(None=Sine(None=Saturate(None=@n17)))
  Add(
   A=LinearInterpolate(
    A=MaterialFunctionCall!(UVsOverride, TextureOverride) {materialFunction=&/MF_Minimap}.RGB
    B=MaterialFunctionCall!(UVsOverride=@n15) {materialFunction=&/MF_Minimap_FOW}.RGB
    Alpha=@n7
   )
   B=Multiply(A=@n1.G, B=Multiply(A=@n16.RGB, B=@n18))
  )
  Add(A=@n14, B=ScalarParameter() {defaultValue=0.002, parameterName=UVdistortion_Ambient})
  Constant() {r=1}
  Constant() {r=1}
  Frac(None=Multiply(A=Time(), B=ScalarParameter() {defaultValue=0.2, parameterName=Speed}))
  MaterialFunctionCall!(Bitmask=ScalarParameter() {defaultValue=2, parameterName=test}) {materialFunction=/Game/Athena/Apollo/Maps/UI/MF_DiscoverabilityPOIUnpack}
  Multiply(A=@n7, B=MaterialFunctionCall() {materialFunction=&MF_Minimap}.Alpha)
  n19 = VectorParameter() {desc=XY location and Z is Radius of the POI, parameterName=LocationRevealFX, defaultValue=<rgba 0.6788 0.6979 0.3 1>}
  Multiply(A=@n19.B, B=@n4)
  n20 = TextureSample(UVs=Multiply(A=TextureCoordinate(), B=ScalarParameter() {defaultValue=2, parameterName=DistortScale1})) {texture=&3xPerlin_M}
  n21 = AppendVector(A=@n20.R, B=@n20.R)
  n22 = ComponentMask(None=VectorParameter() {parameterName=Bias, defaultValue=<rgba 0.25 0.25 0 1>}.RGB) {g=true, r=true}
  n23 = Add(
   A=Multiply(A=Subtract(A=@n21, B=@n22) {constB=0.5}, B=ScalarParameter() {defaultValue=0.1099, parameterName=UVdistortion})
   B=TextureCoordinate()
  )
  n24 = Subtract(A=Distance(A=@n23, B=ComponentMask(None=@n2) {g=true, r=true}), B=@n4)
  OneMinus(None=Saturate(None=Divide(A=@n24) {constB=0.001}))
  n25 = MaterialFunctionCall(Bitmask=CollectionParameter() {collection=&DiscoverabilityMPC, parameterName=AlreadyDiscoveredBitmask}) {materialFunction=&MF_DiscoverabilityPOIUnpack}
  OneMinus(None=@n25.Result)
  n26 = Constant() {r=-1}
  Saturate(None=Subtract(A=MaterialFunctionCall(Bitmask=@n26) {materialFunction=&MF_DiscoverabilityPOIUnpack}.Result, B=@n25.Result))
  Saturate(None=MaterialFunctionCall(Bitmask=CollectionParameter() {collection=&DiscoverabilityMPC, parameterName=FreshlyDiscoveredBitmask}) {materialFunction=&MF_DiscoverabilityPOIUnpack}.Result)
  ScalarParameter() {defaultValue=-0.1, parameterName=OffsetStart}
  TextureSample(UVs=@n23) {texture=&RT_FogofWar}
 ]
}
